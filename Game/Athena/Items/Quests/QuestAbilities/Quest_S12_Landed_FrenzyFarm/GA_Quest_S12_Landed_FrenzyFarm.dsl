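; Gameplay ability for the Season 12 "Landed at Frenzy Farm" quest, triggered by a gameplay event.
; On activation it calls the parent, then compares the owning player pawn's tags with a Frenzy Farm tag.
; On a match it applies the quest's gameplay effect to the owner, then ends the ability.
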(event Ability|EventActivateAbilityFromEvent
  (params (EventData :type Gameplay%20Event%20Data%20Structure%20%28by%20ref%29))
  (|CallParentFunction EventData)
  (if (GameplayTags|TagContainer|CompareTagAssetInterfacetoOtherTags (Variables|Default|GetT_FrenzyFarm) (Variables|GAVars|Data|GetOwningPlayerPawn))
    (Ability|ApplyGameplayEffectToOwner "/Game/Athena/Items/Quests/QuestAbilities/Quest_S12_Landed_FrenzyFarm/GE_Quest_S12_Landed_FrenzyFarm.GE_Quest_S12_Landed_FrenzyFarm_C")
    (else
      _))
  (Ability|EndAbility))
